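# Editor notification sound cue that plays when compilation fails.
# A single wave player feeds the CompileFailed sound wave straight to the root at reduced volume.
# A concurrency override keeps new instances from starting once the limit is reached, and a very large max distance is set.
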SoundCue /Engine/EditorSounds/Notifications/CompileFailed_Cue
root=WavePlayer_108
nodes [{id=WavePlayer_108, properties {soundWaveAssetPtr=&/CompileFailed}}]
properties {bOverrideConcurrency=true, maxDistance=2097152, volumeMultiplier=0.4, concurrencyOverrides {resolutionRule=PreventNew}}
