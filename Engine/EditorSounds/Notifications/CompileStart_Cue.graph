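# CompileStart_Cue is an editor notification Sound Cue, named for the start of a compile.
# A single Wave Player node feeds the output, reusing the CompileSuccess sound wave at raised pitch and reduced volume instead of a dedicated start sound.
# Its own concurrency override rejects new instances once the limit is reached, rather than cutting off those already playing.
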SoundCue /Engine/EditorSounds/Notifications/CompileStart_Cue
root=WavePlayer_109
nodes [{id=WavePlayer_109, properties {soundWaveAssetPtr=&/CompileSuccess}}]
properties {bOverrideConcurrency=true, maxDistance=2097152, pitchMultiplier=1.2, volumeMultiplier=0.4, concurrencyOverrides {resolutionRule=PreventNew}}
