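# CompileSuccess_Cue is an editor notification Sound Cue that plays the CompileSuccess sound.
# A single Wave Player node feeds the output directly at reduced volume, with a very large max distance.
# It overrides concurrency with a Prevent New rule, so new plays are rejected once the instance limit is reached.
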SoundCue /Engine/EditorSounds/Notifications/CompileSuccess_Cue
root=WavePlayer_110
nodes [{id=WavePlayer_110, properties {soundWaveAssetPtr=&/CompileSuccess}}]
properties {bOverrideConcurrency=true, maxDistance=2097152, volumeMultiplier=0.6, concurrencyOverrides {resolutionRule=PreventNew}}
